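# Physics asset for the Street Basketball two-handed pickaxe, giving its backboard and hoop bones capsule bodies for secondary motion.
# The two-capsule board joins the handle through a hinge-like constraint, with twist and one swing axis locked.
# Four progressively narrower hoop capsules chain down from the board, each constrained to the previous with matching swing and twist settings.
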PhysicsAsset /BRCosmetics/Weapons/FORT_Melee/Pickaxe_Street_Basketball/Meshes/SK_Pickaxe_Street_Basketball_PhysicsAsset
bodies {
 Melee_TwoHanded_Handle [{shapeType=Capsule, length=69.5388, radius=3.8386, center=<0.3416 -0.8626 17.4931>, rotation=<rot 0.0661 -90 1.4616>}]
 dyn_board [
  {shapeType=Capsule, length=26.9375, radius=11.5938, center=<-2.8 0.0004 24.5194>, rotation=<rot 0 0 90>}
  {shapeType=Capsule, length=26.9375, radius=11.5938, center=<-2.8 0 0>, rotation=<rot 0 0 90>}
 ]
 dyn_hoop_01 [{shapeType=Capsule, length=1.875, radius=8.125, center=<-2.3935 0 0>, rotation=<rot 90 0 0>}]
 dyn_hoop_02 [{shapeType=Capsule, length=2.0625, radius=7.9375, center=<-2.3935 0 0>, rotation=<rot 90 0 0>}]
 dyn_hoop_03 [{shapeType=Capsule, length=2.75, radius=7.25, center=<-2.3935 0 0>, rotation=<rot 90 0 0>}]
 dyn_hoop_04 [{shapeType=Capsule, length=4.2812, radius=5.7188, center=<-2.3935 0 0>, rotation=<rot 90 0 0>}]
}
constraints [
 {
  bone1Name=dyn_board
  bone2Name=Melee_TwoHanded_Handle
  swing1LimitDegrees=0
  swing1Motion=Locked
  swing2LimitDegrees=45
  twistLimitDegrees=0
  twistMotion=Locked
 }
 {bone1Name=dyn_hoop_01, bone2Name=dyn_board, swing1LimitDegrees=30, swing2LimitDegrees=30, twistLimitDegrees=10}
 {bone1Name=dyn_hoop_02, bone2Name=dyn_hoop_01, swing1LimitDegrees=30, swing2LimitDegrees=30, twistLimitDegrees=10}
 {bone1Name=dyn_hoop_03, bone2Name=dyn_hoop_02, swing1LimitDegrees=30, swing2LimitDegrees=30, twistLimitDegrees=10}
 {bone1Name=dyn_hoop_04, bone2Name=dyn_hoop_03, swing1LimitDegrees=30, swing2LimitDegrees=30, twistLimitDegrees=10}
]
