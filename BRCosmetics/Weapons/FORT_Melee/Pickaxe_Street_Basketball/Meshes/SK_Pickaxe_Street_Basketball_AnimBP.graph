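# Post-style AnimBP for the Street Basketball pickaxe skeleton, adding physics-driven secondary motion to the weapon mesh.
# The input pose passes through a RigidBody node using the pickaxe's override physics asset, with scaled component velocity and acceleration and clamped linear acceleration, then goes to the output.
# AnimDynamics nodes for the board bone and hoop chain are present but unconnected.
# The event graph contains no logic.
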AnimBlueprint /BRCosmetics/Weapons/FORT_Melee/Pickaxe_Street_Basketball/Meshes/SK_Pickaxe_Street_Basketball_AnimBP
target_skeleton=&/SK_Pickaxe_Street_Basketball_Skeleton
graphs [
 AnimGraph {nodes [
   Anim Dynamics - Bone: dyn_board(Alpha=0.35, ExternalForce="0,0,0") {
    angularDampingOverride=0.99
    angularSpringConstant=5000
    bAngularSpring=true
    bGravityOverrideInSimSpace=true
    bOverrideAngularDamping=true
    numSolverIterationsPostUpdate=4
    numSolverIterationsPreUpdate=16
    alphaBoolBlend {blendOption=Linear}
    boundBone {boneName=dyn_board}
    physicsBodyDefinitions [{
      sphereCollisionRadius=0
      boundBone {boneName=dyn_board}
      boxExtents=<10 20 20>
      constraintSetup {angularLimitsMax=<0 10 0>, angularLimitsMin=<0 -10 0>, angularTarget=<1 0 0>}
      localJointOffset=<-3 0 0>
     }]
   }
   "Anim Dynamics - Chain: dyn_hoop_01 -> dyn_hoop_04"(Alpha=1, ExternalForce="0,0,0") {
    angularDampingOverride=0.99
    angularSpringConstant=500
    bAngularSpring=true
    bChain=true
    bGravityOverrideInSimSpace=true
    bOverrideAngularDamping=true
    numSolverIterationsPostUpdate=4
    numSolverIterationsPreUpdate=16
    alphaBoolBlend {blendOption=Linear}
    boundBone {boneName=dyn_hoop_01}
    chainEnd {boneName=dyn_hoop_04}
    physicsBodyDefinitions [{
      sphereCollisionRadius=0
      boundBone {boneName=dyn_hoop_01}
      boxExtents=<10 20 20>
      constraintSetup {angularLimitsMax=<5 60 60>, angularLimitsMin=<-5 -60 -15>, angularTarget=<1 0 0>}
      localJointOffset=<-2 0 0>
     }]
   }
   AnimGraphNode_SubInput_0 = Input Pose()
   n1 = RigidBody(ComponentPose=Local To Component(LocalPose=@AnimGraphNode_SubInput_0)) {
    overridePhysicsAsset=&/SK_Pickaxe_Street_Basketball_PhysicsAsset
    alphaBoolBlend {blendOption=Linear}
    componentAppliedLinearAccClamp=<4000 4000 4000>
    componentLinearAccScale=<2 2 2>
    componentLinearVelScale=<5 5 5>
   }
   Output Pose(Result=Component To Local(ComponentPose=@n1)) {layerGroup=None}
  ]}
 EventGraph {nodes [
   TryGetPawnOwner()
   Event BlueprintUpdateAnimation()
  ]}
]
variables []
